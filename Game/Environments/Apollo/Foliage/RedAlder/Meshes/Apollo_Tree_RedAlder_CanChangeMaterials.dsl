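; Red alder tree Blueprint for the Apollo environment; its name indicates material switching, though no such logic appears in this graph.
; BeginPlay and the construction script only call their parent implementations.
; initRandomUndergrowth seeds a random stream from the truncated sum of actor X and Y, so a location-deterministic coin flip decides whether to add an undergrowth static mesh component.
; Nothing in this graph calls that function.
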(event EventBeginPlay
  (|CallParentFunction))

(fn initRandomUndergrowth
  (if (Math|Random|RandomBoolfromStream (Math|Random|MakeRandomStream (Math|Float|Truncate (+ (.x (Transformation|GetActorLocation)) (.y (Transformation|GetActorLocation))))))
    (AddComponent|NODE_AddStaticMeshComponent-0)))

(fn ConstructionScript
  (|CallParentFunction))
